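# Translucent, unlit, two-sided Niagara ribbon material for the spider-web trail of the Lurk Female pickaxe.
# Noise and web textures pan at a dynamic-parameter speed, blending two reds that the web mask darkens to black before HDR and time-of-day adjustment.
# Opacity combines both textures in a sphere mask with noise-eroded edges, particle alpha, a U-end fade and camera fade.
# Depth fade softens intersections, except on mobile.
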Material /BRCosmetics/Weapons/FORT_Melee/Pickaxe_Lurk_Female/FX/M_Pickaxe_SpiderTrail
graph {
 properties {
  bUsedWithNiagaraRibbons=true
  blendMode=BLEND_Translucent
  pixelDepthOffsetMode=PDOM_Legacy
  shadingModel=MSM_Unlit
  translucencyLightingMode=TLM_VolumetricPerVertexDirectional
  twoSided=true
  usedShadingModels=Unlit
 }
 nodes [
  n1 = Constant() {r=1}
  n2 = Constant() {r=1}
  n3 = Panner(
   Coordinate=Multiply(A=TextureCoordinate() {coordinateIndex=1, uTiling=0.6, vTiling=0.75}, B=AppendVector(A=@n1, B=@n2))
   Speed=AppendVector(A=DynamicParameter() {parameterIndex=1, paramNames [SpiderPanSpeed, Param2, Param3, Param4]}.SpiderPanSpeed, B=Constant())
  ) {speedX=0.6}
  n4 = TextureSample(UVs=@n3) {samplerType=SAMPLERTYPE_Masks, texture=/Game/Weapons/WeaponSkins/Wraps/Textures/Mash/Mash_Noise}
  n5 = Saturate(None=Multiply(A=TextureSample(UVs=@n3) {texture=/Game/Effects/Fort_Effects/Textures/T_BlackWidow_SpiderWeb}.G) {constB=3})
  n6 = LinearInterpolate(
   A=LinearInterpolate(A=Constant3Vector() {constant=<rgba 0.2865 0.0254 0.0254 0>}, B=Constant3Vector() {constant=<rgba 0.6927 0.1168 0.0685 0>}, Alpha=@n4.R)
   Alpha=@n5
  ) {constB=0}
  n7 = Clamp(None=@n6)
  n8 = Constant() {r=1}
  n9 = Constant() {r=1}
  MaterialFunctionCall(Emissive=Multiply(
    A=MaterialFunctionCall() {materialFunction=&Color_Time_of_Day}.Result
    B=MaterialFunctionCall(LDR Input=@n7, HDR Dynamic Range=@n8, HDR Tint=@n6, HDR Intensity Multiplier=@n9) {materialFunction=&DeriveHDRfromLDR}.RemappedRange
   )) {materialFunction=&MF_FX_EmissiveAdjustments}
  n10 = SphereMask(A=TextureCoordinate() {uTiling=0}, B=Constant2Vector() {g=0.5}, Radius=Constant() {r=0.5}, Hardness=Constant())
  Multiply(A=@n1, B=Add(A=@n10))
  n11 = Multiply(A=@n4.R, B=@n10)
  n12 = Multiply(A=@n5, B=@n10)
  n13 = Constant()
  n14 = Constant() {r=0.2}
  n15 = Subtract(A=Multiply(A=@n10) {constB=2}, B=@n4.R)
  n16 = Multiply(A=ParticleColor().A, B=Multiply(A=Add(A=@n11, B=@n12), B=SmoothStep(Min=@n13, Max=@n14, Value=@n15)))
  n17 = SmoothStep(Min=Constant(), Max=Constant() {r=0.25}, Value=ComponentMask(None=TextureCoordinate()) {r=true})
  n18 = ScalarParameter() {defaultValue=64, parameterName=Fade Length}
  n19 = ScalarParameter() {defaultValue=24, parameterName=Fade Offset}
  n20 = Multiply(
   A=Multiply(
    A=Multiply(A=@n16, B=@n17)
    B=MaterialFunctionCall(Fade Length=@n18, Fade Offset=@n19) {materialFunction=&CameraDepthFade}.Result
   )
   B=Constant() {r=1.5}
  )
  ShadingPathSwitch(Default=DepthFade(Opacity=@n20, FadeDistance=ScalarParameter() {defaultValue=25, parameterName=FadeDistance}), Mobile=@n20) {desc=Migrated from FeatureLevelSwitch}
 ]
}
