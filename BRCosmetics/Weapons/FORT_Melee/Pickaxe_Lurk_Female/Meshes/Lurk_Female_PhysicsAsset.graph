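# Physics asset for the female Lurk pickaxe, a two-handed melee weapon cosmetic in Fortnite Battle Royale.
# It holds a single capsule body on the Melee_TwoHanded_Handle bone, offset and rotated relative to that bone.
# There are no constraints, so the weapon has one collision shape with no jointed or secondary physics parts.
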PhysicsAsset /BRCosmetics/Weapons/FORT_Melee/Pickaxe_Lurk_Female/Meshes/Lurk_Female_PhysicsAsset
bodies {Melee_TwoHanded_Handle [{shapeType=Capsule, length=53.4591, radius=17.5444, center=<-3.7002 -2.2695 22.6307>, rotation=<rot 1.23 -90 0.7171>}]}
constraints []
